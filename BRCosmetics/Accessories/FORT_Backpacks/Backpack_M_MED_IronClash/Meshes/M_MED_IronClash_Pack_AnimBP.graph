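# Animation Blueprint for the IronClash backpack that nudges its pack bone to follow parachute state and the wearer's spine bend.
# A first Modify Bone node lifts the pack in world space, blending in and out sinusoidally with bIsParachuteOpened.
# A second adds a roll and offset, scaled by rotationSpine_inMidAir.
# Each update derives that alpha from the X-axis quaternion twist of spine_05 against pelvis on the owner's skeletal mesh.
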AnimBlueprint /BRCosmetics/Accessories/FORT_Backpacks/Backpack_M_MED_IronClash/Meshes/M_MED_IronClash_Pack_AnimBP
target_skeleton=&/M_MED_IronClash_Pack_Skeleton
graphs [
 AnimGraph {nodes [
   n1 = "Transform (Modify) Bone - Bone: pack"!(
    ComponentPose=Local To Component(LocalPose=Input Pose())
    bAlphaBoolEnabled:bool=Get bIsParachuteOpened()
    Translation="0,0,4"
    Rotation="0,0,0"
    Scale
   ) -> (Pose) {
    alphaInputType=Bool
    translationMode=BMM_Additive
    translationSpace=BCS_WorldSpace
    alphaBoolBlend {blendInTime=1, blendOption=Sinusoidal, blendOutTime=2}
    boneToModify {boneName=pack}
    translation {z=4}
   }
   n2 = "Transform (Modify) Bone - Bone: pack"!(
    ComponentPose=@n1
    Alpha:float=Get rotationSpine_inMidAir() -> (rotationSpine_inMidAir:float)
    Translation="0.000,-2.297,0.000"
    Rotation="-0.000,0.000,17.122"
    Scale
   ) -> (Pose) {
    rotationMode=BMM_Additive
    translationMode=BMM_Additive
    translationSpace=BCS_WorldSpace
    alphaBoolBlend {blendInTime=1, blendOption=Sinusoidal, blendOutTime=2}
    boneToModify {boneName=pack}
    rotation {roll=17.122}
    translation {y=-2.297}
   }
   Output Pose(Result=Component To Local(ComponentPose=@n2)) {layerGroup=None}
  ]}
 EventGraph {nodes [
   Event BlueprintUpdateAnimation()
   > Set rotationSpine_inMidAir(rotationSpine_inMidAir:float=GetAlphaFromReferenceTransformDelta(
     SkeletalMeshComponent=GetOwnerSkeletalMeshComponent()
     BoneName=spine_05
     BaseName=pelvis
     DeltaComponentType=QuaternionTwist
     InputRange=(X=-40,Y=-55)
     TwistAxis="1,0,0"
    )) -> (Output_Get:float)
   TryGetPawnOwner()
  ]}
]
variables [rotationSpine_inMidAir {type=number}]
